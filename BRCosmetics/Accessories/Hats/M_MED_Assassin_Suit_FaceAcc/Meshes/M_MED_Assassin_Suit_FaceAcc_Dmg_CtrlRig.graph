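# Control Rig for the Assassin Suit face accessory cosmetic (Dmg variant) that adjusts rear and side hair bones as the head moves during emotes or crouching.
# A sequence runs two spherical pose readers on the head relative to spine_05, each with its own target orientation.
# Their outputs, multiplied by IsPlayingEmote and IsCrouching respectively, weight identical transform offsets on five dyn_hair bones.
# The crouch reader has debug drawing left on.
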ControlRigBlueprint /BRCosmetics/Accessories/Hats/M_MED_Assassin_Suit_FaceAcc/Meshes/M_MED_Assassin_Suit_FaceAcc_Dmg_CtrlRig
graphs [
 RigVMFunctionLibrary {links [], nodes []}
 RigVMModel {nodes [
   RigUnit_BeginExecution = &RigUnit_BeginExecution()
   RigVMFunction_Sequence = &RigVMFunction_Sequence(ExecuteContext=@RigUnit_BeginExecution.ExecutePin)
   n1 = &RigUnit_SphericalPoseReader(
    ActiveRegionScaleFactors=(PositiveWidth=1,NegativeWidth=1,PositiveHeight=1,NegativeHeight=1)
    ActiveRegionSize=0.05
    Debug=(bDrawDebug=false,bDraw2D=False,bDrawLocalAxes=False,DebugScale=25,DebugSegments=20,DebugThickness=0.25)
    DriverAxis=(X=1,Y=0,Z=0)
    DriverItem=(Type=Bone,Name="head")
    FalloffRegionScaleFactors=(PositiveWidth=1,NegativeWidth=1,PositiveHeight=0.5,NegativeHeight=1)
    FalloffSize=0.3
    FlipHeightScaling=False
    FlipWidthScaling=False
    OptionalParentItem=(Type=Bone,Name="spine_05")
    RotationOffset=(X=36,Y=-92,Z=-68)
    ExecutePin=@RigVMFunction_Sequence.A
   )
   Multiply = &RigVMFunction_MathDoubleMul(A=@n1.OutputParam, B=RigVMVariableNode() {variable=IsPlayingEmote}.Value)
   > &RigUnit_ModifyTransforms(
    ItemToModify=((Item=(Type=Bone,Name="dyn_hair_bk_l_01"),Transform=(Rotation=(X=0,Y=-0.1736,Z=0,W=0.9848),Translation=(X=0,Y=0,Z=0),Scale3D=(X=1,Y=1,Z=1))))
    WeightMinimum=0
    Weight=@Multiply.Result
   )
   > &RigUnit_ModifyTransforms(
    ItemToModify=((Item=(Type=Bone,Name="dyn_hair_l_bk_02"),Transform=(Rotation=(X=0,Y=-0.0872,Z=0,W=0.9962),Translation=(X=0,Y=0,Z=0),Scale3D=(X=1,Y=1,Z=1))))
    WeightMinimum=0
    Weight=@Multiply.Result
   )
   > &RigUnit_ModifyTransforms(
    ItemToModify=((Item=(Type=Bone,Name="dyn_hair_bk_01"),Transform=(Rotation=(X=0,Y=-0.1736,Z=0,W=0.9848),Translation=(X=0,Y=0,Z=0),Scale3D=(X=1,Y=1,Z=1))))
    WeightMinimum=0
    Weight=@Multiply.Result
   )
   > &RigUnit_ModifyTransforms(
    ItemToModify=((Item=(Type=Bone,Name="dyn_hair_bk_r_01"),Transform=(Rotation=(X=0,Y=-0.1736,Z=0,W=0.9848),Translation=(X=-2,Y=0,Z=0),Scale3D=(X=1,Y=1,Z=1))))
    WeightMinimum=0
    Weight=@Multiply.Result
   )
   > &RigUnit_ModifyTransforms(
    ItemToModify=((Item=(Type=Bone,Name="dyn_hair_r_bk_01"),Transform=(Rotation=(X=0,Y=-0.0872,Z=0,W=0.9962),Translation=(X=0,Y=0,Z=0),Scale3D=(X=1,Y=1,Z=1))))
    WeightMinimum=0
    Weight=@Multiply.Result
   )
   n2 = &RigUnit_SphericalPoseReader(
    ActiveRegionScaleFactors=(PositiveWidth=1,NegativeWidth=1,PositiveHeight=1,NegativeHeight=1)
    ActiveRegionSize=0.05
    Debug=(bDrawDebug=true,bDraw2D=False,bDrawLocalAxes=False,DebugScale=25,DebugSegments=20,DebugThickness=0.25)
    DriverAxis=(X=1,Y=0,Z=0)
    DriverItem=(Type=Bone,Name="head")
    FalloffRegionScaleFactors=(PositiveWidth=1,NegativeWidth=1,PositiveHeight=1,NegativeHeight=1)
    FalloffSize=0.3
    FlipHeightScaling=False
    FlipWidthScaling=False
    OptionalParentItem=(Type=Bone,Name="spine_05")
    RotationOffset=(X=18,Y=-72,Z=-80)
    ExecutePin=@RigVMFunction_Sequence.B
   )
   n3 = &RigVMFunction_MathDoubleMul(A=@n2.OutputParam, B=RigVMVariableNode() {variable=IsCrouching}.Value)
   > &RigUnit_ModifyTransforms(
    ItemToModify=((Item=(Type=Bone,Name="dyn_hair_bk_l_01"),Transform=(Rotation=(X=0,Y=-0.1736,Z=0,W=0.9848),Translation=(X=0,Y=0,Z=0),Scale3D=(X=1,Y=1,Z=1))))
    WeightMinimum=0
    Weight=@n3.Result
   )
   > &RigUnit_ModifyTransforms(
    ItemToModify=((Item=(Type=Bone,Name="dyn_hair_l_bk_02"),Transform=(Rotation=(X=0,Y=-0.0872,Z=0,W=0.9962),Translation=(X=0,Y=0,Z=0),Scale3D=(X=1,Y=1,Z=1))))
    WeightMinimum=0
    Weight=@n3.Result
   )
   > &RigUnit_ModifyTransforms(
    ItemToModify=((Item=(Type=Bone,Name="dyn_hair_bk_01"),Transform=(Rotation=(X=0,Y=-0.1736,Z=0,W=0.9848),Translation=(X=0,Y=0,Z=0),Scale3D=(X=1,Y=1,Z=1))))
    WeightMinimum=0
    Weight=@n3.Result
   )
   > &RigUnit_ModifyTransforms(
    ItemToModify=((Item=(Type=Bone,Name="dyn_hair_bk_r_01"),Transform=(Rotation=(X=0,Y=-0.1736,Z=0,W=0.9848),Translation=(X=-2,Y=0,Z=0),Scale3D=(X=1,Y=1,Z=1))))
    WeightMinimum=0
    Weight=@n3.Result
   )
   > &RigUnit_ModifyTransforms(
    ItemToModify=((Item=(Type=Bone,Name="dyn_hair_r_bk_01"),Transform=(Rotation=(X=0,Y=-0.0872,Z=0,W=0.9962),Translation=(X=0,Y=0,Z=0),Scale3D=(X=1,Y=1,Z=1))))
    WeightMinimum=0
    Weight=@n3.Result
   )
  ]}
]
hierarchy [
 root {loc=<0 0 0>, rot=<0 0 0 1>}
 pelvis {parent=root, loc=<0 -0.7071 93.1191>, rot=<0.0127 0.707 -0.0127 -0.707>}
 spine_01 {parent=pelvis, loc=<6.7494 -0.6388 0>, rot=<0 0 0.034 -0.9994>}
 spine_02 {parent=spine_01, loc=<7.2518 -0.035 0>, rot=<0 0 0.0238 -0.9997>}
 spine_03 {parent=spine_02, loc=<6.0104 0 0>, rot=<0 0 0.0149 -0.9999>}
 spine_04 {parent=spine_03, loc=<8.1024 0 0>, rot=<0 0 -0.0132 -0.9999>}
 spine_05 {parent=spine_04, loc=<10.0146 0 0>, rot=<0 0 -0.0015 -1>}
 clavicle_l {parent=spine_05, loc=<4.3575 -0.1466 -2.3223>, rot=<0.0465 -0.7289 -0.009 -0.6829>}
 clavicle_r {parent=spine_05, loc=<4.3575 -0.1466 2.3223>, rot=<0.7289 0.0465 -0.6829 0.009>}
 neck_01 {parent=spine_05, loc=<11.3138 0.1529 0>, rot=<0 0 -0.2003 -0.9797>}
 neck_02 {parent=neck_01, loc=<4.5627 0 0>, rot=<0 0 0.0504 -0.9987>}
 head {parent=neck_02, loc=<5.3914 0 0>, rot=<0 0 0.115 -0.9934>}
 faceAttach {parent=head, loc=<-2.586 0.9279 0.003>, rot=<0.0082 -0.7071 -0.0082 -0.7071>, scale=<0.89 0.89 0.89>}
 C_brow_mid {parent=faceAttach, loc=<-0.0232 9.9477 8.2627>, rot=<0 0.0315 -0.0062 -0.9995>}
 L_brow_mid {parent=faceAttach, loc=<3.1829 9.4412 9.1257>, rot=<-0.0002 -0.0048 0.2553 -0.9669>}
 L_brow_outer {parent=faceAttach, loc=<5.1258 7.3788 9.4466>, rot=<0.0627 0.075 0.4258 -0.8995>}
 L_cheek_inner {parent=faceAttach, loc=<4.1482 7.7271 3.4147>, rot=<-0.0278 -0.1611 0.3312 -0.9293>}
 C_lip_upper_mid {parent=faceAttach, loc=<0.011 9.2299 0.7271>, rot=<0.2225 0.003 -0.0011 -0.9749>}
 L_lip_upper_outer {parent=faceAttach, loc=<1.6448 8.5587 0.721>, rot=<0.0573 -0.0123 0.2394 -0.9692>}
 L_lip_corner {parent=faceAttach, loc=<2.3746 8.0051 0.5356>, rot=<0.0024 0.0086 -0.0146 -0.9999>}
 C_nose_bridge {parent=faceAttach, loc=<0.0183 11.0878 5.4808>, rot=<-0.2217 -0.0239 -0.0067 -0.9748>}
 R_brow_mid {parent=faceAttach, loc=<-3.1762 9.4425 9.1325>, rot=<-0.0001 0.0043 -0.2553 -0.9669>}
 R_brow_outer {parent=faceAttach, loc=<-5.1284 7.3657 9.4305>, rot=<0.0415 -0.0483 -0.4321 -0.8996>}
 R_cheek_inner {parent=faceAttach, loc=<-4.1311 7.7353 3.4043>, rot=<-0.0269 0.1589 -0.3314 -0.9296>}
 R_lip_corner {parent=faceAttach, loc=<-2.3854 8.0095 0.5548>, rot=<0.0028 -0.0081 0.0146 -0.9999>}
 C_jaw {parent=faceAttach, loc=<-0.0349 3.0539 1.204>, rot=<0 0 0 1>}
 L_lip_lower_outer {parent=C_jaw, loc=<1.6478 5.3424 -0.4354>, rot=<0.0082 0.0188 0.286 -0.958>}
 C_lip_lower_mid {parent=C_jaw, loc=<0.0343 5.94 -0.4384>, rot=<0 0.0048 0.001 -1>}
 R_lip_lower_outer {parent=C_jaw, loc=<-1.5784 5.4493 -0.4553>, rot=<-0.0018 0.0358 -0.2435 -0.9692>}
 R_lip_upper_outer {parent=faceAttach, loc=<-1.6667 8.6134 0.7402>, rot=<0.0679 -0.0181 -0.2272 -0.9713>}
 dyn_hair_fr_l_01 {parent=head, loc=<10.9661 6.7793 -4.7284>, rot=<0.0805 -0.9657 -0.2326 -0.0828>}
 dyn_hair_fr_l_02 {parent=dyn_hair_fr_l_01, loc=<6.2415 0 0>, rot=<0 0 0 1>}
 dyn_hair_fr_l_03 {parent=dyn_hair_fr_l_01, loc=<2.6719 1.986 0.5785>, rot=<0 0 0 1>}
 dyn_hair_fr_l_fr_01 {parent=head, loc=<7.6783 2.6099 -7.4396>, rot=<0.0041 -0.9999 -0.0111 0>}
 dyn_hair_fr_l_fr_02 {parent=dyn_hair_fr_l_fr_01, loc=<4.988 0 0>, rot=<0 0 0 1>}
 dyn_hair_l_bk_01 {parent=head, loc=<5.5004 -1.6205 -7.2442>, rot=<0.0041 -0.9972 0.0741 -0.0003>}
 dyn_hair_l_bk_02 {parent=dyn_hair_l_bk_01, loc=<5.4783 -0.7145 0.1359>, rot=<0 0.1348 0 -0.9909>}
 dyn_hair_l_bk_03 {parent=dyn_hair_l_bk_01, loc=<4.9732 2.0245 -0.6162>, rot=<0 0 0 1>}
 dyn_hair_bk_l_01 {parent=head, loc=<3.9464 -5.4791 -4.6307>, rot=<0.0039 -0.933 0.3599 -0.0015>}
 dyn_hair_bk_l_02 {parent=dyn_hair_bk_l_01, loc=<4.988 0 0>, rot=<0 0.2407 0 -0.9706>}
 dyn_hair_bk_01 {parent=head, loc=<3.7692 -7.0401 0.3236>, rot=<-0.0169 -0.7069 0.7069 0.0169>}
 dyn_hair_bk_02 {parent=dyn_hair_bk_01, loc=<4.988 0 0>, rot=<0 0.2407 0 -0.9706>}
 dyn_hair_bk_r_01 {parent=head, loc=<3.9464 -5.4791 4.6307>, rot=<0.0015 -0.3599 0.933 -0.0039>}
 dyn_hair_bk_r_02 {parent=dyn_hair_bk_r_01, loc=<4.988 0 0>, rot=<0 0.2407 0 -0.9706>}
 dyn_hair_r_bk_01 {parent=head, loc=<5.5004 -1.6205 7.2442>, rot=<0.0003 -0.0741 0.9972 -0.0041>}
 dyn_hair_r_bk_02 {parent=dyn_hair_r_bk_01, loc=<5.3542 -1.198 -0.1703>, rot=<0 0.1348 0 -0.9909>}
 dyn_hair_r_bk_03 {parent=dyn_hair_r_bk_01, loc=<5.8417 1.3844 0.6162>, rot=<0 0 0 1>}
 dyn_hair_r_fr_01 {parent=head, loc=<7.6783 2.6099 7.4396>, rot=<0 0.0111 0.9999 -0.0041>}
 dyn_hair_r_fr_02 {parent=dyn_hair_r_fr_01, loc=<4.988 0 0>, rot=<0 0 0 1>}
 dyn_hair_fr_r_01 {parent=head, loc=<10.9661 6.7793 4.7284>, rot=<0.0828 0.2326 0.9657 -0.0805>}
 dyn_hair_fr_r_02 {parent=dyn_hair_fr_r_01, loc=<6.6592 1.2305 -0.6142>, rot=<0.0754 0.0077 0.101 -0.992>}
 dyn_hair_fr_r_03 {parent=dyn_hair_fr_r_01, loc=<5.2103 -0.5485 0.2821>, rot=<0 0 0 1>}
 dyn_hair_fr_01 {parent=head, loc=<11.6957 8.7945 1.1092>, rot=<0.4611 0.0898 0.8823 0.0309>}
 dyn_hair_fr_02 {parent=dyn_hair_fr_01, loc=<3.1961 0 0>, rot=<0.0079 -0.384 0.0136 -0.9232>}
 dyn_hair_fr_03 {parent=dyn_hair_fr_02, loc=<4.5042 0 0>, rot=<0 0 0 1>}
 dyn_strand_bk_l_01 {parent=head, loc=<13.4649 -3.0239 -4.405>, rot=<0.0625 -0.9324 0.3158 -0.1643>}
 dyn_strand_bk_l_02 {parent=dyn_strand_bk_l_01, loc=<6.4028 0 0.1311>, rot=<0 -0.1593 0 -0.9872>}
 dyn_strand_bk_l_03 {parent=dyn_strand_bk_l_02, loc=<6.4716 0 0>, rot=<0 0.132 0 -0.9913>}
 dyn_strand_bk_01 {parent=head, loc=<13.2315 -5.251 -0.8418>, rot=<0.1113 -0.6983 0.6983 -0.1113>}
 dyn_strand_bk_02 {parent=dyn_strand_bk_01, loc=<6.4028 0 0.1311>, rot=<0 -0.1216 0 -0.9926>}
 dyn_strand_bk_03 {parent=dyn_strand_bk_02, loc=<6.4716 0 0>, rot=<0 0.132 0 -0.9913>}
 dyn_strand_bk_r_01 {parent=head, loc=<13.4448 -1.8378 4.7414>, rot=<0.1952 -0.2957 0.9285 -0.1113>}
 dyn_strand_bk_r_02 {parent=dyn_strand_bk_r_01, loc=<6.4028 0 0.1311>, rot=<0 -0.1923 0 -0.9813>}
 dyn_strand_bk_r_03 {parent=dyn_strand_bk_r_02, loc=<6.4716 0 0>, rot=<0 0.132 0 -0.9913>}
 dyn_string_bk_01 {parent=head, loc=<15.8153 -2.9107 0.7057>, rot=<0.2424 -0.6643 0.6643 -0.2424>}
 dyn_string_bk_02 {parent=dyn_string_bk_01, loc=<6.4028 0 0.1311>, rot=<0.0068 -0.3031 0.0213 -0.9527>}
 dyn_string_bk_03 {parent=dyn_string_bk_02, loc=<6.4716 0 0>, rot=<0 0.132 0 -0.9913>}
 dyn_string_l_01 {parent=head, loc=<15.3921 5.1608 -1.6642>, rot=<0.1291 -0.865 -0.0757 -0.4789>}
 dyn_string_l_02 {parent=dyn_string_l_01, loc=<4.5785 0 0>, rot=<-0.0659 -0.313 -0.0218 -0.9472>}
 dyn_string_l_03 {parent=dyn_string_l_02, loc=<6.4716 0 0>, rot=<0 -0.1516 0 -0.9884>}
 dyn_string_r_01 {parent=head, loc=<13.557 8.4392 1.176>, rot=<0.4807 0.1218 0.8376 -0.2293>}
 dyn_string_r_02 {parent=dyn_string_r_01, loc=<4.5785 0 0>, rot=<0 -0.3085 0 -0.9512>}
 dyn_string_r_03 {parent=dyn_string_r_02, loc=<6.4716 0 0>, rot=<0.0177 -0.1834 0.0943 -0.9784>}
 dyn_string_r_04 {parent=dyn_string_r_03, loc=<4.4172 0 0>, rot=<0 0 0.1746 -0.9846>}
 dyn_string_fr_r_01 {parent=head, loc=<10.9568 8.2058 2.4779>, rot=<0.3006 0.2694 0.9108 -0.087>}
 dyn_string_fr_r_02 {parent=dyn_string_fr_r_01, loc=<2.9708 0 0>, rot=<-0.0796 -0.223 0.1855 -0.9537>}
 dyn_string_fr_r_03 {parent=dyn_string_fr_r_02, loc=<3.5573 0 0>, rot=<0 -0.0772 0 -0.997>}
 dyn_string_fr_l_01 {parent=head, loc=<10.7798 9.4146 -1.3618>, rot=<0.2224 -0.8362 -0.4082 -0.2911>}
 dyn_string_fr_l_02 {parent=dyn_string_fr_l_01, loc=<4.5785 0 0>, rot=<-0.1305 -0.2428 -0.0942 -0.9566>}
]
variables [
 IsCrouching {type=double, default=0}
 IsPlayingEmote {type=double, default=0}
]
