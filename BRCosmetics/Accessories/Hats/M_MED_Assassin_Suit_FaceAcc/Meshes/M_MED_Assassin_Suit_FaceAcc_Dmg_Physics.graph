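# Physics asset for the Assassin suit face accessory (Dmg variant), defining secondary-motion physics for its hair, strand and string bones.
# Chains of small capsules and a few boxes hang from the head, with head, neck, upper spine and clavicles supplying larger body volumes.
# Swing and twist limits are generally tighter on strands and strings than hair, while body joints rotate freely.
# Several third hair bones constrain to their chain's first bone, skipping the second.
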PhysicsAsset /BRCosmetics/Accessories/Hats/M_MED_Assassin_Suit_FaceAcc/Meshes/M_MED_Assassin_Suit_FaceAcc_Dmg_Physics
bodies {
 clavicle_l [{shapeType=Capsule, length=11.0192, radius=6.8871, center=<9.4266 -1.1806 -2.4977>, rotation=<rot 69.3297 1.4476 176.9462>}]
 clavicle_r [{shapeType=Capsule, length=11.0192, radius=6.8871, center=<-9.4266 1.1806 2.4977>, rotation=<rot 69.3297 1.4476 -3.0538>}]
 dyn_hair_bk_01 [{shapeType=Box, extentX=6.1875, extentY=5.625, extentZ=0.5, center=<2.5972 0.0001 0.1602>, rotation=<rot 8.2819 0 0>}]
 dyn_hair_bk_02 [{
   shapeType=Capsule
   length=3.6684
   radius=2.0006
   shapeName=dyn_hair_bk_02_capsule
   center=<3.3161 0 -0.5728>
   rotation=<rot -30.0001 0.0003 -90.0006>
  }]
 dyn_hair_bk_l_01 [{shapeType=Box, extentX=6.25, extentY=4.6875, extentZ=1, center=<3.0788 0 -0.2051>}]
 dyn_hair_bk_l_02 [{
   shapeType=Capsule
   length=1
   radius=1.75
   shapeName=dyn_hair_bk_l_02_capsule
   center=<3.2575 0 -0.9234>
   rotation=<rot 64.7729 -0.0001 89.9997>
  }]
 dyn_hair_bk_r_01 [{
   shapeType=Box
   extentX=5
   extentY=0.5
   extentZ=5.95
   shapeName=dyn_hair_bk_r_01_box
   center=<2.8714 -0.6174 0.3708>
   rotation=<rot 6.2891 -87.1578 -89.3508>
  }]
 dyn_hair_bk_r_02 [{
   shapeType=Capsule
   length=0.5
   radius=2
   shapeName=dyn_hair_bk_r_02_capsule
   center=<3.011 -1.1678 -1.1137>
   rotation=<rot 63.055 12.4924 101.1727>
  }]
 dyn_hair_fr_01 [{shapeType=Capsule, length=1.25, radius=1, shapeName=dyn_hair_fr_01_capsule, center=<1.7029 0 0.2902>, rotation=<rot -89.4536 -0.0007 0.0005>}]
 dyn_hair_fr_02 [{shapeType=Capsule, length=3.7, radius=1, shapeName=dyn_hair_fr_02_capsule, center=<3 0 0.0924>, rotation=<rot -89.3 -180 180>}]
 dyn_hair_fr_03 [{shapeType=Capsule, length=3, radius=1, shapeName=dyn_hair_fr_03_capsule, center=<3.0829 0 0>, rotation=<rot 88.8864 -179.9995 -179.9998>}]
 dyn_hair_fr_l_01 [{
   shapeType=Capsule
   length=4.25
   radius=1.5
   shapeName=dyn_hair_fr_l_01_capsule
   center=<3.1132 0 0.6224>
   rotation=<rot -88.8617 0.0007 -0.0007>
  }]
 dyn_hair_fr_l_02 [{
   shapeType=Capsule
   length=2.75
   radius=1.75
   shapeName=dyn_hair_fr_l_02_capsule
   center=<3.0372 -0.0001 0.1918>
   rotation=<rot 89.524 -0.0008 -0.0006>
  }]
 dyn_hair_fr_l_03 [{
   shapeType=Capsule
   length=4.75
   radius=1
   shapeName=dyn_hair_fr_l_03_capsule
   center=<3.0733 0 0.2976>
   rotation=<rot -88.9242 -0.0006 0.0004>
  }]
 dyn_hair_fr_l_fr_01 [{
   shapeType=Capsule
   length=3
   radius=1.25
   shapeName=dyn_hair_fr_l_fr_01_capsule
   center=<3.0502 -0.0001 -0.0374>
   rotation=<rot -85.7008 -0.0002 0.0002>
  }]
 dyn_hair_fr_l_fr_02 [{
   shapeType=Capsule
   length=2.75
   radius=1.5
   shapeName=dyn_hair_fr_l_fr_02_capsule
   center=<3.0633 0 -0.4423>
   rotation=<rot -89.8881 -0.0001 0.0001>
  }]
 dyn_hair_fr_r_01 [{shapeType=Capsule, length=4, radius=1, shapeName=dyn_hair_fr_r_01_capsule, center=<2.9691 0 0.1444>, rotation=<rot -89.3173 -0.0031 0.003>}]
 dyn_hair_fr_r_02 [{
   shapeType=Capsule
   length=4.2151
   radius=0.7849
   shapeName=dyn_string_bk_01_capsule
   center=<2.1918 0 0>
   rotation=<rot -87.8975 179.9999 -179.9999>
  }]
 dyn_hair_fr_r_03 [{
   shapeType=Capsule
   length=3.7096
   radius=0.6362
   shapeName=dyn_string_bk_01_capsule
   center=<2.8445 0.093 0>
   rotation=<rot -89.4182 179.9976 165.8308>
  }]
 dyn_hair_l_bk_01 [{shapeType=Capsule, length=0, radius=0.5, shapeName=dyn_hair_l_bk_01_capsule, center=<3.1774 -0.0001 0>, rotation=<rot -90 0.0198 -0.0199>}]
 dyn_hair_l_bk_02 [{
   shapeType=Capsule
   length=2.75
   radius=1.25
   shapeName=dyn_hair_l_bk_02_capsule
   center=<2.7959 0 -0.8177>
   rotation=<rot -73.6846 179.9998 -179.9999>
  }]
 dyn_hair_l_bk_03 [{shapeType=Capsule, length=4.25, radius=1.25, shapeName=dyn_hair_l_bk_03_capsule, center=<3.1709 0 0.5355>, rotation=<rot -90 -0.0198 0.0197>}]
 dyn_hair_r_bk_01 [{
   shapeType=Capsule
   length=0
   radius=0.5
   shapeName=dyn_string_bk_01_capsule
   center=<2.8139 0 1.3536>
   rotation=<rot -89.6103 0.0001 -0.0001>
  }]
 dyn_hair_r_bk_02 [{
   shapeType=Capsule
   length=2.5
   radius=1.5
   shapeName=dyn_string_bk_01_capsule
   center=<3.4652 0 0.43>
   rotation=<rot -73.3534 -179.9999 179.9999>
  }]
 dyn_hair_r_bk_03 [{shapeType=Capsule, length=2, radius=1.5, shapeName=dyn_string_bk_01_capsule, center=<2.5057 0 0.4488>, rotation=<rot -89.567 -0.0016 0.0016>}]
 dyn_hair_r_fr_01 [{shapeType=Box, extentX=6.75, extentY=2.1875, extentZ=0.5, center=<3.0788 0 -0.0001>}]
 dyn_hair_r_fr_02 [{shapeType=Box, extentX=6.25, extentY=2.1875, extentZ=0.5, center=<3.0788 0 -0.0001>}]
 dyn_strand_bk_01 [{
   shapeType=Capsule
   length=4.75
   radius=1.25
   shapeName=dyn_string_bk_01_capsule
   center=<3.3654 0 0.3558>
   rotation=<rot -89.6215 0.0001 -0.0001>
  }]
 dyn_strand_bk_02 [{shapeType=Capsule, length=3.75, radius=1.25, shapeName=dyn_string_bk_01_capsule, center=<2.5994 0 -0.021>, rotation=<rot -86.1358 0 0>}]
 dyn_strand_bk_03 [{shapeType=Capsule, length=6.25, radius=1.25, shapeName=dyn_string_bk_01_capsule, center=<2.1246 0 -0.5688>, rotation=<rot -73.4858 -180 180>}]
 dyn_strand_bk_l_01 [{
   shapeType=Capsule
   length=3.5
   radius=1.5
   shapeName=dyn_string_bk_01_capsule
   center=<3.1974 0 0.2975>
   rotation=<rot -87.6963 0.0007 -0.0007>
  }]
 dyn_strand_bk_l_02 [{
   shapeType=Capsule
   length=3.5
   radius=2
   shapeName=dyn_string_bk_01_capsule
   center=<2.7596 -0.0001 0>
   rotation=<rot -89.7668 179.9909 -179.991>
  }]
 dyn_strand_bk_l_03 [{
   shapeType=Capsule
   length=3
   radius=3.25
   shapeName=dyn_string_bk_01_capsule
   center=<2.8182 0 -0.5975>
   rotation=<rot -71.9614 -179.9999 179.9999>
  }]
 dyn_strand_bk_r_01 [{shapeType=Capsule, length=3, radius=1, shapeName=dyn_strand_bk_r_01_capsule, center=<3.1327 0 0.8561>, rotation=<rot -90 -1.6876 1.6879>}]
 dyn_strand_bk_r_02 [{shapeType=Capsule, length=3, radius=1, shapeName=dyn_strand_bk_r_02_capsule, center=<2.8763 0 0.8562>, rotation=<rot -90 -0.2373 0.2375>}]
 dyn_strand_bk_r_03 [{shapeType=Capsule, length=3, radius=1, shapeName=dyn_strand_bk_r_03_capsule, center=<2.931 0 -0.6351>, rotation=<rot -73.1248 -180 179.9998>}]
 dyn_string_bk_01 [{
   shapeType=Capsule
   length=6.0159
   radius=1
   shapeName=dyn_string_bk_01_capsule
   center=<4.7516 0 -0.7135>
   rotation=<rot -79.7336 -179.9999 179.9999>
  }]
 dyn_string_bk_02 [{shapeType=Capsule, length=4.2659, radius=1.2341, shapeName=dyn_string_bk_01_capsule, center=<4.923 0 -0.235>, rotation=<rot -79.8435 0 0>}]
 dyn_string_bk_03 [{shapeType=Capsule, length=3.7659, radius=0.9841, shapeName=dyn_string_bk_01_capsule, center=<4.6479 0 -0.62>, rotation=<rot 69.4054 0 0>}]
 dyn_string_fr_l_01 [{
   shapeType=Capsule
   length=3
   radius=1
   shapeName=dyn_string_fr_l_01_capsule
   center=<2.6147 0.0001 0.013>
   rotation=<rot -89.7161 -0.0144 0.0144>
  }]
 dyn_string_fr_l_02 [{
   shapeType=Capsule
   length=3.75
   radius=1
   shapeName=dyn_string_fr_l_02_capsule
   center=<2.8915 0.0001 -0.174>
   rotation=<rot -86.5489 179.9988 -179.9988>
  }]
 dyn_string_fr_r_01 [{shapeType=Capsule, length=1.75, radius=1, shapeName=dyn_string_fr_r_01_capsule, center=<1.6492 0 0>, rotation=<rot -89.1635 0.0041 -0.0041>}]
 dyn_string_fr_r_02 [{
   shapeType=Capsule
   length=2.25
   radius=1
   shapeName=dyn_string_fr_r_02_capsule
   center=<2.5232 0 -0.002>
   rotation=<rot -89.7188 0.0012 -0.0013>
  }]
 dyn_string_fr_r_03 [{
   shapeType=Capsule
   length=2.25
   radius=1
   shapeName=dyn_string_fr_r_03_capsule
   center=<2.6669 0 -0.0019>
   rotation=<rot -89.7324 -0.009 0.009>
  }]
 dyn_string_l_01 [{shapeType=Capsule, length=0, radius=0.5, center=<2.3609 -0.0001 0>, rotation=<rot 90 0 0>}]
 dyn_string_l_02 [{shapeType=Capsule, length=3, radius=1.5, center=<3.3613 -0.0001 0.2666>, rotation=<rot 90 0 0>}]
 dyn_string_l_03 [{shapeType=Capsule, length=3, radius=1.5, center=<2.3609 -0.0001 0>, rotation=<rot 90 0 0>}]
 dyn_string_r_01 [{
   shapeType=Capsule
   length=0
   radius=0.5
   shapeName=dyn_string_bk_01_capsule
   center=<2.5416 0.0001 0.2352>
   rotation=<rot -89.371 179.9973 -179.9973>
  }]
 dyn_string_r_02 [{
   shapeType=Capsule
   length=3
   radius=1
   shapeName=dyn_string_bk_01_capsule
   center=<3.5046 0.0003 0.0302>
   rotation=<rot -81.9299 -179.9994 179.9992>
  }]
 dyn_string_r_03 [{shapeType=Capsule, length=2.5, radius=1, shapeName=dyn_string_bk_01_capsule, center=<2.5369 0 -0.2225>, rotation=<rot -86.5513 0 0>}]
 dyn_string_r_04 [{
   shapeType=Capsule
   length=2.5
   radius=1
   shapeName=dyn_string_bk_01_capsule
   center=<3.2147 0.0001 -0.258>
   rotation=<rot -87.7132 180 -179.9999>
  }]
 head [
  {shapeType=Capsule, length=8.4375, radius=7.0938, center=<3.8064 0.5099 0.0003>, rotation=<rot 90 180 175>}
  {shapeType=Capsule, length=6.2812, radius=5.8438, center=<5.0855 2.7731 0.1733>, rotation=<rot 90 0 0>}
  {shapeType=Capsule, length=6.2812, radius=5.3438, center=<5.083 4.6248 0.1733>, rotation=<rot 90 0 11.2502>}
  {shapeType=Capsule, length=5.5, radius=4.5, shapeName=head_capsule, center=<9.7329 1.9412 -1.0149>, rotation=<rot 3.1831 15.2253 -101.5366>}
 ]
 neck_02 [{shapeType=Capsule, length=7.4375, radius=6.3175, center=<-3.2837 0.2194 0>, rotation=<rot 90 180 172.9092>}]
 spine_05 [
  {shapeType=Capsule, length=8.0226, radius=15.695, center=<-1.9981 1.8109 0>}
  {shapeType=Capsule, length=13.9518, radius=13.9425, center=<-0.0102 0.4452 0>}
 ]
}
constraints [
 {bone1Name=dyn_hair_fr_l_02, bone2Name=dyn_hair_fr_l_01, swing1LimitDegrees=45, swing2LimitDegrees=45, twistLimitDegrees=5}
 {bone1Name=dyn_hair_fr_l_03, bone2Name=dyn_hair_fr_l_01, swing1LimitDegrees=45, swing2LimitDegrees=45, twistLimitDegrees=5}
 {bone1Name=dyn_hair_fr_l_fr_02, bone2Name=dyn_hair_fr_l_fr_01, swing1LimitDegrees=45, swing2LimitDegrees=45, twistLimitDegrees=5}
 {bone1Name=dyn_hair_l_bk_02, bone2Name=dyn_hair_l_bk_01, swing1LimitDegrees=45, swing2LimitDegrees=45, twistLimitDegrees=5}
 {bone1Name=dyn_hair_l_bk_03, bone2Name=dyn_hair_l_bk_01, swing1LimitDegrees=45, swing2LimitDegrees=45, twistLimitDegrees=5}
 {bone1Name=dyn_hair_bk_l_02, bone2Name=dyn_hair_bk_l_01, swing1LimitDegrees=45, swing2LimitDegrees=45, twistLimitDegrees=5}
 {bone1Name=dyn_hair_bk_02, bone2Name=dyn_hair_bk_01, swing1LimitDegrees=45, swing2LimitDegrees=45, twistLimitDegrees=5}
 {bone1Name=dyn_hair_bk_r_02, bone2Name=dyn_hair_bk_r_01, swing1LimitDegrees=45, swing2LimitDegrees=45, twistLimitDegrees=5}
 {bone1Name=dyn_hair_r_fr_02, bone2Name=dyn_hair_r_fr_01, swing1LimitDegrees=45, swing2LimitDegrees=45, twistLimitDegrees=5}
 {bone1Name=dyn_hair_fr_r_02, bone2Name=dyn_hair_fr_r_01, swing1LimitDegrees=45, swing2LimitDegrees=45, twistLimitDegrees=5}
 {bone1Name=dyn_hair_fr_r_03, bone2Name=dyn_hair_fr_r_01, swing1LimitDegrees=45, swing2LimitDegrees=45, twistLimitDegrees=5}
 {bone1Name=dyn_hair_fr_02, bone2Name=dyn_hair_fr_01, swing1LimitDegrees=45, swing2LimitDegrees=45, twistLimitDegrees=5}
 {bone1Name=dyn_hair_fr_03, bone2Name=dyn_hair_fr_02, swing1LimitDegrees=45, swing2LimitDegrees=45, twistLimitDegrees=5}
 {bone1Name=dyn_strand_bk_l_02, bone2Name=dyn_strand_bk_l_01, swing1LimitDegrees=10, swing2LimitDegrees=10, twistLimitDegrees=5}
 {bone1Name=dyn_strand_bk_l_03, bone2Name=dyn_strand_bk_l_02, swing1LimitDegrees=25, swing2LimitDegrees=25, twistLimitDegrees=5}
 {bone1Name=dyn_strand_bk_02, bone2Name=dyn_strand_bk_01, swing1LimitDegrees=5, swing2LimitDegrees=10, twistLimitDegrees=5}
 {bone1Name=dyn_strand_bk_03, bone2Name=dyn_strand_bk_02, swing1LimitDegrees=30, swing2LimitDegrees=45, twistLimitDegrees=5}
 {bone1Name=dyn_string_bk_02, bone2Name=dyn_string_bk_01, swing1LimitDegrees=15, swing2LimitDegrees=30, twistLimitDegrees=5}
 {bone1Name=dyn_string_bk_03, bone2Name=dyn_string_bk_02, swing1LimitDegrees=15, swing2LimitDegrees=20, twistLimitDegrees=5}
 {bone1Name=dyn_string_l_02, bone2Name=dyn_string_l_01, swing1LimitDegrees=15, swing2LimitDegrees=10, twistLimitDegrees=10}
 {bone1Name=dyn_string_l_03, bone2Name=dyn_string_l_02, swing1LimitDegrees=20, swing2LimitDegrees=30, twistLimitDegrees=5}
 {bone1Name=dyn_string_r_02, bone2Name=dyn_string_r_01, swing1LimitDegrees=20, swing2LimitDegrees=20, twistLimitDegrees=0}
 {bone1Name=dyn_string_r_03, bone2Name=dyn_string_r_02, swing1LimitDegrees=20, swing2LimitDegrees=20, twistLimitDegrees=0}
 {bone1Name=dyn_string_r_04, bone2Name=dyn_string_r_03, swing1LimitDegrees=20, swing2LimitDegrees=20, twistLimitDegrees=0}
 {bone1Name=dyn_string_fr_r_02, bone2Name=dyn_string_fr_r_01, swing1LimitDegrees=45, swing2LimitDegrees=45, twistLimitDegrees=5}
 {bone1Name=dyn_string_fr_r_03, bone2Name=dyn_string_fr_r_02, swing1LimitDegrees=45, swing2LimitDegrees=45, twistLimitDegrees=5}
 {bone1Name=dyn_string_fr_l_02, bone2Name=dyn_string_fr_l_01, swing1LimitDegrees=45, swing2LimitDegrees=45, twistLimitDegrees=0}
 {bone1Name=dyn_hair_fr_l_01, bone2Name=head, swing1LimitDegrees=45, swing2LimitDegrees=45, twistLimitDegrees=5}
 {bone1Name=dyn_hair_fr_l_fr_01, bone2Name=head, swing1LimitDegrees=45, swing2LimitDegrees=45, twistLimitDegrees=5}
 {bone1Name=dyn_hair_l_bk_01, bone2Name=head, swing1LimitDegrees=45, swing2LimitDegrees=45, twistLimitDegrees=5}
 {bone1Name=dyn_hair_bk_l_01, bone2Name=head, swing1LimitDegrees=45, swing2LimitDegrees=45, twistLimitDegrees=5}
 {bone1Name=dyn_hair_bk_01, bone2Name=head, swing1LimitDegrees=45, swing2LimitDegrees=45, twistLimitDegrees=5}
 {bone1Name=dyn_hair_r_bk_01, bone2Name=head, swing1LimitDegrees=45, swing2LimitDegrees=45, twistLimitDegrees=5}
 {bone1Name=dyn_hair_r_fr_01, bone2Name=head, swing1LimitDegrees=45, swing2LimitDegrees=45, twistLimitDegrees=5}
 {bone1Name=dyn_hair_fr_r_01, bone2Name=head, swing1LimitDegrees=45, swing2LimitDegrees=45, twistLimitDegrees=5}
 {bone1Name=dyn_hair_fr_01, bone2Name=head, swing1LimitDegrees=45, swing2LimitDegrees=45, twistLimitDegrees=5}
 {bone1Name=dyn_strand_bk_l_01, bone2Name=head, swing1LimitDegrees=5, swing2LimitDegrees=5, twistLimitDegrees=5}
 {bone1Name=dyn_strand_bk_01, bone2Name=head, swing1LimitDegrees=5, swing2LimitDegrees=15, twistLimitDegrees=5}
 {bone1Name=dyn_string_bk_01, bone2Name=head, swing1LimitDegrees=5, swing2LimitDegrees=25, twistLimitDegrees=5}
 {bone1Name=dyn_string_l_01, bone2Name=head, swing1LimitDegrees=5, swing2LimitDegrees=15, twistLimitDegrees=5}
 {bone1Name=dyn_string_r_01, bone2Name=head, swing1LimitDegrees=20, swing2LimitDegrees=20, twistLimitDegrees=0}
 {bone1Name=dyn_string_fr_r_01, bone2Name=head, swing1LimitDegrees=20, swing2LimitDegrees=45, twistLimitDegrees=5}
 {bone1Name=dyn_string_fr_l_01, bone2Name=head, swing1LimitDegrees=30, swing2LimitDegrees=30, twistLimitDegrees=0}
 {
  bone1Name=head
  bone2Name=neck_02
  swing1LimitDegrees=45
  swing1Motion=Free
  swing2LimitDegrees=45
  swing2Motion=Free
  twistLimitDegrees=45
  twistMotion=Free
 }
 {
  bone1Name=neck_02
  bone2Name=spine_05
  swing1LimitDegrees=45
  swing1Motion=Free
  swing2LimitDegrees=45
  swing2Motion=Free
  twistLimitDegrees=45
  twistMotion=Free
 }
 {bone1Name=dyn_strand_bk_r_03, bone2Name=dyn_strand_bk_r_02, swing1LimitDegrees=20, swing2LimitDegrees=20, twistLimitDegrees=0}
 {bone1Name=dyn_strand_bk_r_02, bone2Name=dyn_strand_bk_r_01, swing1LimitDegrees=20, swing2LimitDegrees=20, twistLimitDegrees=0}
 {bone1Name=dyn_strand_bk_r_01, bone2Name=head, swing1LimitDegrees=20, swing2LimitDegrees=20, twistLimitDegrees=0}
 {
  bone1Name=clavicle_l
  bone2Name=spine_05
  swing1LimitDegrees=45
  swing1Motion=Free
  swing2LimitDegrees=45
  swing2Motion=Free
  twistLimitDegrees=45
  twistMotion=Free
 }
 {
  bone1Name=clavicle_r
  bone2Name=spine_05
  swing1LimitDegrees=45
  swing1Motion=Free
  swing2LimitDegrees=45
  swing2Motion=Free
  twistLimitDegrees=45
  twistMotion=Free
 }
 {bone1Name=dyn_hair_r_bk_02, bone2Name=dyn_hair_r_bk_01, swing1LimitDegrees=45, swing2LimitDegrees=45, twistLimitDegrees=0}
 {bone1Name=dyn_hair_r_bk_03, bone2Name=dyn_hair_r_bk_01, swing1LimitDegrees=45, swing2LimitDegrees=45, twistLimitDegrees=5}
 {bone1Name=dyn_hair_bk_r_01, bone2Name=head, swing1LimitDegrees=45, swing2LimitDegrees=45, twistLimitDegrees=5}
]
