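# Physics asset for the M_MED_Assassin_Suit_FaceAcc cosmetic, covering its dynamic back and side hair bones.
# Five two-body hair chains, mostly boxes, hang from a head capsule, with a spine_05 capsule representing the upper torso.
# Hair joints limit swing and keep twist tight, while the head-to-spine joint leaves all angular motion free.
# The right-back chain links bone 01 directly to 03, with no body for an 02 bone.
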PhysicsAsset /BRCosmetics/Accessories/Hats/M_MED_Assassin_Suit_FaceAcc/Meshes/M_MED_Assassin_Suit_FaceAcc_Dyn
bodies {
 dyn_hair_bk_01 [{shapeType=Box, extentX=6, extentY=6, extentZ=2, center=<2.3422 0 0.5126>, rotation=<rot 4.9995 0 0>}]
 dyn_hair_bk_02 [{shapeType=Capsule, length=5.5, radius=2, shapeName=dyn_hair_bk_02_capsule, center=<2.6281 0 0.4098>, rotation=<rot -0.0006 0 -90>}]
 dyn_hair_bk_l_01 [{shapeType=Box, extentX=6, extentY=4, extentZ=2, center=<3.0788 0 0.2825>}]
 dyn_hair_bk_l_02 [{shapeType=Box, extentX=4, extentY=4, extentZ=2, center=<2.4129 0 -0.1638>, rotation=<rot 0 0.0002 -0.0001>}]
 dyn_hair_bk_r_01 [{shapeType=Box, extentX=6, extentY=4, extentZ=2, center=<2.6724 0 0.5395>}]
 dyn_hair_bk_r_02 [{shapeType=Box, extentX=5, extentY=4, extentZ=3, center=<2.7148 0 0>}]
 dyn_hair_l_bk_01 [{shapeType=Box, extentX=6, extentY=5, extentZ=2.8603, center=<3 0 0.7225>}]
 dyn_hair_l_bk_02 [{shapeType=Box, extentX=3.1551, extentY=4.5625, extentZ=1.789, center=<3.3567 0 -0.3137>, rotation=<rot 30 0.0001 0>}]
 dyn_hair_r_bk_01 [{shapeType=Box, extentX=6, extentY=4, extentZ=2, center=<4.0371 0 0.3258>}]
 dyn_hair_r_bk_03 [{shapeType=Box, extentX=4, extentY=6, extentZ=2, center=<2.9601 -1.5789 0.0531>, rotation=<rot 20 -0.0003 0>}]
 head [{shapeType=Capsule, length=9.5625, radius=6.5938, center=<0.6322 0.2429 0.0002>, rotation=<rot 90 -162.646 -167.646>}]
 spine_05 [{shapeType=Capsule, length=7.5675, radius=9.63, center=<6.9085 0.864 0>, rotation=<rot 90 -165.9638 159.0372>}]
}
constraints [
 {bone1Name=dyn_hair_l_bk_02, bone2Name=dyn_hair_l_bk_01, swing1LimitDegrees=20, swing2LimitDegrees=45, twistLimitDegrees=5}
 {bone1Name=dyn_hair_bk_l_02, bone2Name=dyn_hair_bk_l_01, swing1LimitDegrees=45, swing2LimitDegrees=45, twistLimitDegrees=5}
 {bone1Name=dyn_hair_bk_02, bone2Name=dyn_hair_bk_01, swing1LimitDegrees=45, swing2LimitDegrees=45, twistLimitDegrees=5}
 {bone1Name=dyn_hair_bk_r_02, bone2Name=dyn_hair_bk_r_01, swing1LimitDegrees=45, swing2LimitDegrees=45, twistLimitDegrees=5}
 {bone1Name=dyn_hair_r_bk_03, bone2Name=dyn_hair_r_bk_01, swing1LimitDegrees=45, swing2LimitDegrees=45, twistLimitDegrees=5}
 {bone1Name=dyn_hair_l_bk_01, bone2Name=head, swing1LimitDegrees=20, swing2LimitDegrees=45, twistLimitDegrees=5}
 {bone1Name=dyn_hair_bk_l_01, bone2Name=head, swing1LimitDegrees=45, swing2LimitDegrees=45, twistLimitDegrees=5}
 {bone1Name=dyn_hair_bk_01, bone2Name=head, swing1LimitDegrees=45, swing2LimitDegrees=45, twistLimitDegrees=5}
 {bone1Name=dyn_hair_bk_r_01, bone2Name=head, swing1LimitDegrees=45, swing2LimitDegrees=45, twistLimitDegrees=5}
 {bone1Name=dyn_hair_r_bk_01, bone2Name=head, swing1LimitDegrees=45, swing2LimitDegrees=45, twistLimitDegrees=5}
 {
  bone1Name=head
  bone2Name=spine_05
  swing1LimitDegrees=45
  swing1Motion=Free
  swing2LimitDegrees=45
  swing2Motion=Free
  twistLimitDegrees=45
  twistMotion=Free
 }
]
